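# Niagara system for the StarBeamBW vehicle boost cosmetic, emitting a star-beam trail from the vehicle's boost sockets.
# Two short-lived sprite emitters split particles across sockets, one with scattered backward velocity, the other streaming straight back.
# A per-port ribbon emitter with five renderers inherits fading source movement and moves faster in frontend preview.
# Painted-variant overrides set ribbon colors, which curves animate with layer widths through custom particle attributes.
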NiagaraSystem /VehicleCosmetics/Boosts/Boost_StarBeamBW/FX/NS_Boost_StarBeamBW
dependencies {
 usedDataInterfaces [&NiagaraDataInterfaceCurve, &NiagaraDataInterfaceVectorCurve, &NiagaraDataInterfaceSkeletalMesh]
 usedDynamicInputs []
 usedModules [
  &self:SystemSpawnScript.SystemScriptSource.SystemScriptGraph.NiagaraNodeAssignment_0.SetVariables_#1
  &SystemState
 ]
}
emitters {
 NE_VC_LocalSprites {
  emitterUpdateScript [
   &EmitterState()
   &VC_FilteredSocketCount()
   &SpawnRate(SpawnRate=60, Use Spawn Probability=false)
  ]
  particleSpawnScript [
   &VC_PaintedVariantsOverride() {enabled=false}
   &InitializeParticle(
    Lifetime Max=0.3
    Lifetime Min=0.1
    Lifetime Mode=::NewEnumerator1
    Sprite Rotation Mode=::NewEnumerator2
    Sprite Size Mode=::NewEnumerator1
    Uniform Sprite Size Max=20
    Uniform Sprite Size Min=10
    UsePositionOffset=false
    Write Material Random=true
   )
   &PartitionParticles(Number Of Partitions=$Emitter.PortCount)
   &SkeletalMeshLocation(
    Bone / Socket Index=$Particles.PartitionParticles.Partition
    Bone / Socket Sampling Mode=::NewEnumerator5
    MaskSpawnGroup=false
    Mesh Sampling Type=::NewEnumerator6
    Skeletal Mesh=$System.SkeletalMesh
   )
   &AddVelocity(Velocity=&RandomRangeVector(Minimum=<-750 -300 -200>, Maximum=<-375 300 200>))
   &ShapeLocation(MaskSpawnGroup=false, Override Seed=false, Sphere Radius=25, UseOwnerScale=true)
   /Game/Effects/Niagara/Modules/Sprite/FN_SpriteUV(End Frame=2)
   /VehicleCosmetics/FX/VC_FXParents/Modules/DLR_ProductPreviews(GameModeID:NiagaraInt32=$System.GameModeID, bPreviewVehicleSpeed:NiagaraBool=$System.bPreviewVehicleSpeed)
  ]
  particleUpdateScript [
   &ParticleState()
   &ScaleSpriteSize(Uniform Curve Sprite Scale=DataInterface() {propertyValues {Curve {keys [{tangentMode=RCTM_User, time=0, value=0}, <at 0.2 1>, <at 0.75 1>, {tangentMode=RCTM_User, time=1, value=0}]}}})
   &SpriteRotationRate(Rotation Rate=&RandomRangeFloat(Minimum=1, Maximum=2))
   &InheritSourceMovement(
    Applied Position Scale=&Multiply_VectorByFloat(
     Vector=<1 1 1>
     Float=&FloatFromCurve(FloatCurve=DataInterface() {propertyValues {Curve {
         defaultValue=1
         keys [
          {arriveTangent=-1.7885, interpMode=RCIM_Cubic, leaveTangent=-1.7885, tangentMode=RCTM_User, time=0.05, value=1}
          {interpMode=RCIM_Cubic, tangentMode=RCTM_User, time=1, value=0}
         ]
        }}})
    )
    ApplyForce=false
    ApplyPosition=true
    ApplyPositionFalloff=false
   ) {enabled=false, module=InheritSourceMovement001}
   &ScaleColor(
    Scale RGB=&VectorFromCurve(VectorCurve=DataInterface() {propertyValues {XCurve {keys [<at 0 0>,<at 0.3 10>,<at 1 3>]}, YCurve {keys [<at 0 0>,<at 0.3 10>,<at 1 3>]}, ZCurve {keys [<at 0 0>,<at 0.3 10>,<at 1 3>]}}})
    ScaleA=true
    ScaleRGB=true
   )
   &SolveForcesAndVelocity(Clamp Velocity=false, Limit Acceleration=false)
  ]
  rendererClasses [&NiagaraSpriteRendererProperties]
  renderers [{rendererClass=&NiagaraSpriteRendererProperties}]
 }
 NE_VC_Ribbons {
  emitterUpdateScript [
   &EmitterState()
   &VC_RateQualityHandler(High Spawn Rate=30, Low Spawn Rate=10, Medium Spawn Rate=20, Use Custom Rate?=true) {enabled=false, module=VC_RateQualityHandler001}
   &VC_FilteredSocketCount()
   &VC_RateQualityHandler()
   &VC_SpawnRibbon_v1(Number of Ribbons=$Emitter.PortCount, Spawn Time=-1, SpawnRate=$Emitter.Spawn Rate)
  ]
  particleSpawnScript [
   &VC_PaintedVariantsOverride(
    Black=<rgba 0.07 0.07 0.07 3>
    BurntSienna=<rgba 0.15 0.0463 0.0185 0.9>
    Cobalt=<rgba 0.45 0.5417 1 0.1>
    Crimson=<rgba 1 0 0.0271 0.5>
    DefaultAlpha=0.6
    DefaultColor=<rgba 0.001 0.001 0.001 1>
    ForestGreen=<rgba 0 0.5 0.0253 0.6>
    Grey=<rgba 0.2031 0.2031 0.2031 0.8>
    Lime=<rgba 0.375 0.5 0 0.8>
    Orange=<rgba 0.672 0.3657 0.2016 1>
    Pink=<rgba 1 0.15 0.575 0.8>
    Purple=<rgba 0.5133 0.14 0.7 1>
    R1 Black:LinearColor=<rgba 0.2 0.2 0.2 50>
    R1 BurntSienna:LinearColor=<rgba 0.2 0.1212 0.1 40>
    R1 Cobalt:LinearColor=<rgba 0 0.0925 5 10>
    R1 Crimson:LinearColor=<rgba 2 0 0.0542 11>
    R1 DefaultColor=<rgba 1 1 1 15>
    R1 ForestGreen:LinearColor=<rgba 0 1.5 0 20>
    R1 Grey:LinearColor=<rgba 0.4 0.4 0.4 35>
    R1 Lime:LinearColor=<rgba 0.6667 1 0 40>
    R1 Orange:LinearColor=<rgba 1 0.75 0 40>
    R1 Pink:LinearColor=<rgba 1 0.45 0.8 80>
    R1 Purple:LinearColor=<rgba 1 0.5 0.8182 40>
    R1 Saffron:LinearColor=<rgba 2 1.5333 0 30>
    R1 SkyBlue:LinearColor=<rgba 0.4 1.4646 2 15>
    R1 TitaniumWhite:LinearColor=<rgba 2 2 2 15>
    R2 Black:LinearColor=<rgba 0.098 0.098 0.098 1>
    R2 BurntSienna:LinearColor=<rgba 0.12 0.0774 0.066 1>
    R2 Cobalt:LinearColor=<rgba 0 0.0925 5 0.2>
    R2 Crimson:LinearColor=<rgba 1 0.15 0.173 0.5>
    R2 DefaultColor:LinearColor=<rgba 0 0 0 4>
    R2 ForestGreen:LinearColor=<rgba 0.15 1 0.1929 0.35>
    R2 Grey:LinearColor=<rgba 0.3 0.3 0.3 1>
    R2 Lime:LinearColor=<rgba 0.6667 1 0 0.6>
    R2 Orange:LinearColor=<rgba 1 0.5833 0 1>
    R2 Pink:LinearColor=<rgba 1 0.5 0.8182 0.5>
    R2 Purple:LinearColor=<rgba 1 0.3 0.7455 1>
    R2 Saffron:LinearColor=<rgba 1.5 1.15 0 1>
    R2 SkyBlue:LinearColor=<rgba 0.12 0.8386 1.2 1>
    Saffron=<rgba 1.2 0.948 0.12 1>
    SkyBlue=<rgba 0 0.5988 0.9 1>
    TitaniumWhite=<rgba 0.5 0.5 0.5 1>
    Use R1 Painted Overrides=true
    Use R2 Painted Overrides=true
   )
   &VC_InitializeRibbon_v1(
    Color=$Particles.Color
    Custom Offset=true
    Enable Skeletal Position=true
    Lifetime=1
    Local Vector Offset=&AddVectorToPosition(Position=<2 0 0>, Vector=$User.DTCarOffset)
    Offset Amount=1
    Ribbon Width=40
   )
   &AddVelocity(
    Velocity=&MakeCustomVectorFromBool(Bool=$User.bIsFrontendPreview, True Vector=<-2000 0 1>, False Vector=<-1200 0 0>)
    Velocity Speed Scale=&MakeFloatFromVector(
     Vector=&VectorFromFloat(Value=&FloatFromCurve(FloatCurve=DataInterface() {propertyValues {Curve {keys [{interpMode=RCIM_Cubic, tangentMode=RCTM_User, time=0, value=1}, {interpMode=RCIM_Cubic, tangentMode=RCTM_User, time=1, value=0}]}}}))
     Channel=::NewEnumerator0
    )
   ) {module=AddVelocity001}
   &DLR_ProductPreviews(GameModeID:NiagaraInt32=$System.GameModeID, bPreviewVehicleSpeed:NiagaraBool=$System.bPreviewVehicleSpeed)
  ]
  particleUpdateScript [
   &ParticleState()
   &InheritSourceMovement(
    Applied Position Scale=&VectorFromFloat(Value=&FloatFromCurve(FloatCurve=DataInterface() {propertyValues {Curve {keys [<at 0 1>,<at 1 0>]}}}))
    ApplyForce=false
    ApplyPosition=true
    ApplyPositionFalloff=false
   ) {module=InheritSourceMovement001}
   &ScaleRibbonWidth(Ribbon Width Scale=&FloatFromCurve(FloatCurve=DataInterface() {propertyValues {Curve {keys [<at 0 1>,<at 1 0.3>]}}}))
   &InheritSourceMovement(
    Applied Position Scale=&VectorFromFloat(Value=&FloatFromCurve(FloatCurve=DataInterface() {propertyValues {Curve {
         defaultValue=1
         keys [
          {arriveTangent=-1.7885, interpMode=RCIM_Cubic, leaveTangent=-1.7885, tangentMode=RCTM_User, time=0.05, value=1}
          {interpMode=RCIM_Cubic, tangentMode=RCTM_User, time=1, value=0}
         ]
        }}}))
    ApplyForce=false
    ApplyPosition=true
    ApplyPositionFalloff=false
   ) {enabled=false}
   &SolveForcesAndVelocity(Clamp Velocity=false, Limit Acceleration=false)
   &ScaleColor(
    Scale Alpha=&FloatFromCurve(FloatCurve=DataInterface() {propertyValues {Curve {defaultValue=1, keys [{tangentMode=RCTM_User, time=0, value=1}, {tangentMode=RCTM_User, time=1, value=1.5}]}}})
    ScaleA=true
    ScaleRGB=true
   )
   /SetVariables(Particles.ScaleOutLinecolor:LinearColor=&MultiplyLinearColor(
     Linear Color=$Particles.R1 Color
     Scale Factor=&MakeVector4FromVectorAndFloat(
      "Vector (RGB)"=<1 1 1>
      "Float (Alpha)"=&FloatFromCurve(
       FloatCurve=DataInterface() {propertyValues {Curve {
          defaultValue=1
          keys [
           {arriveTangent=-2.1875, interpMode=RCIM_Cubic, time=0, value=0}
           {interpMode=RCIM_Cubic, tangentMode=RCTM_User, time=0.2, value=1}
           {interpMode=RCIM_Cubic, tangentMode=RCTM_User, time=0.5, value=-0.05}
          ]
         }}}
       Scale Curve=3
      )
     )
    )) {isSetParameters=true}
   /SetVariables(Particles.ScaleColorNoise:LinearColor=&MultiplyLinearColor(
     Linear Color=$Particles.R2 Color
     Scale Factor=&MakeVector4FromVectorAndFloat(
      "Vector (RGB)"=<1 1 1>
      "Float (Alpha)"=&FloatFromCurve(FloatCurve=DataInterface() {propertyValues {Curve {
          defaultValue=1
          keys [
           {arriveTangent=5, interpMode=RCIM_Cubic, time=0, value=0}
           {interpMode=RCIM_Cubic, tangentMode=RCTM_User, time=0.2, value=1}
           {interpMode=RCIM_Cubic, tangentMode=RCTM_User, time=0.5, value=0}
          ]
         }}})
     )
    )) {isSetParameters=true}
   /SetVariables(
    Particles.NoiseBluecolor:LinearColor=&MultiplyLinearColor(
     Linear Color=$Particles.R3 Color
     Scale Factor=&MakeVector4FromVectorAndFloat(
      "Vector (RGB)"=<1 1 1>
      "Float (Alpha)"=&FloatFromCurve(
       FloatCurve=DataInterface() {propertyValues {Curve {
          defaultValue=1
          keys [{interpMode=RCIM_Cubic, tangentMode=RCTM_User, time=0.3, value=1}, {interpMode=RCIM_Cubic, tangentMode=RCTM_User, time=0.55, value=0}]
         }}}
       Scale Curve=0.4
      )
     )
    )
    Particles.Size_NoiseFinal:NiagaraFloat=&Multiply_Float(
     A=25
     B=&FloatFromCurve(FloatCurve=DataInterface() {propertyValues {Curve {keys [{arriveTangent=0.4294, interpMode=RCIM_Cubic, time=0.4717, value=0.9872}, {interpMode=RCIM_Cubic, tangentMode=RCTM_User, time=1, value=0.3}]}}})
    )
   ) {isSetParameters=true}
   /SetVariables(Particles.SizeOutlines:NiagaraFloat=&Multiply_Float(
     A=60
     B=&FloatFromCurve(FloatCurve=DataInterface() {propertyValues {Curve {keys [<at 0 1>,<at 1 0.5>]}}})
    )) {isSetParameters=true}
   /SetVariables(Particles.float:NiagaraFloat=&Multiply_Float(
     A=35
     B=&FloatFromCurve(FloatCurve=DataInterface() {propertyValues {Curve {keys [<at 0 1>,<at 1 0.3>]}}})
    )) {isSetParameters=true}
   &VC_UpdateRibbon_v1(Enable Skeletal Position=true)
  ]
  rendererClasses [&NiagaraRibbonRendererProperties]
  renderers [
   {rendererClass=&NiagaraRibbonRendererProperties}
   {rendererClass=&NiagaraRibbonRendererProperties, rendererIndex=1}
   {rendererClass=&NiagaraRibbonRendererProperties, rendererIndex=2}
   {rendererClass=&NiagaraRibbonRendererProperties, rendererIndex=3}
   {rendererClass=&NiagaraRibbonRendererProperties, rendererIndex=4}
  ]
 }
 NE_VC_TrailSprites {
  emitterUpdateScript [
   &EmitterState()
   &VC_FilteredSocketCount()
   &SpawnRate(SpawnRate=30, Use Spawn Probability=false)
  ]
  particleSpawnScript [
   &VC_PaintedVariantsOverride() {enabled=false}
   &InitializeParticle(
    Lifetime Max=0.4
    Lifetime Min=0.25
    Lifetime Mode=::NewEnumerator1
    Sprite Rotation Mode=::NewEnumerator2
    Sprite Size Mode=::NewEnumerator1
    Uniform Sprite Size Max=14
    Uniform Sprite Size Min=6
    UsePositionOffset=false
    Write Material Random=true
   )
   &PartitionParticles(Number Of Partitions=$Emitter.PortCount)
   &SkeletalMeshLocation(
    Bone / Socket Index=$Particles.PartitionParticles.Partition
    Bone / Socket Sampling Mode=::NewEnumerator5
    MaskSpawnGroup=false
    Mesh Sampling Type=::NewEnumerator6
    Skeletal Mesh=$System.SkeletalMesh
   )
   &AddVelocity(Velocity=<-1500 0 50>)
   &ShapeLocation(MaskSpawnGroup=false, Override Seed=false, Sphere Radius=25, UseOwnerScale=true)
   &FN_SpriteUV(End Frame=2)
   &DLR_ProductPreviews(GameModeID:NiagaraInt32=$System.GameModeID, bPreviewVehicleSpeed:NiagaraBool=$System.bPreviewVehicleSpeed)
  ]
  particleUpdateScript [
   &ParticleState()
   &ScaleSpriteSize(Uniform Curve Sprite Scale=DataInterface() {propertyValues {Curve {keys [{tangentMode=RCTM_User, time=0, value=0}, <at 0.2 1>, <at 0.75 1>, {tangentMode=RCTM_User, time=1, value=0}]}}})
   &SpriteRotationRate(Rotation Rate=&RandomRangeFloat(Minimum=1, Maximum=2))
   &InheritSourceMovement(
    Applied Position Scale=&Multiply_VectorByFloat(
     Vector=<1 1 1>
     Float=&FloatFromCurve(FloatCurve=DataInterface() {propertyValues {Curve {
         defaultValue=1
         keys [
          {arriveTangent=-1.7885, interpMode=RCIM_Cubic, leaveTangent=-1.7885, tangentMode=RCTM_User, time=0.05, value=1}
          {interpMode=RCIM_Cubic, tangentMode=RCTM_User, time=1, value=0}
         ]
        }}})
    )
    ApplyForce=false
    ApplyPosition=true
    ApplyPositionFalloff=false
   ) {enabled=false, module=InheritSourceMovement001}
   &ScaleColor(
    Scale RGB=&VectorFromCurve(VectorCurve=DataInterface() {propertyValues {XCurve {keys [<at 0 0>,<at 0.3 10>,<at 1 3>]}, YCurve {keys [<at 0 0>,<at 0.3 10>,<at 1 3>]}, ZCurve {keys [<at 0 0>,<at 0.3 10>,<at 1 3>]}}})
    ScaleA=true
    ScaleRGB=true
   )
   &SolveForcesAndVelocity(Clamp Velocity=false, Limit Acceleration=false)
  ]
  rendererClasses [&NiagaraSpriteRendererProperties]
  renderers [{rendererClass=&NiagaraSpriteRendererProperties}]
 }
}
summary {
 systemName=NS_Boost_StarBeamBW
 emitters [
  {emitterName=NE_VC_LocalSprites, rendererClasses [&NiagaraSpriteRendererProperties]}
  {emitterName=NE_VC_TrailSprites, rendererClasses [&NiagaraSpriteRendererProperties]}
  {emitterName=NE_VC_Ribbons, rendererClasses [&NiagaraRibbonRendererProperties]}
 ]
 userVariables [
  User.bIsFrontendPreview {type=NiagaraBool, defaultValue=true}
  User.bIsPainted {type=NiagaraBool}
  User.ColorVariantID {type=NiagaraInt32}
  User.ColorWhite {type=NiagaraFloat, defaultValue=0, description=""}
  User.DTCarOffset {type=Vector3f, description="", defaultValue {struct=Vector3f, value=<0 0 0>}}
  User.EmissiveTrail {type=NiagaraFloat, defaultValue=1, description=""}
  User.EmissiveTrailAlpha {type=NiagaraFloat, defaultValue=0.6, description=""}
  User.PaintedColor {type=LinearColor, description="", defaultValue {struct=LinearColor, value=<rgba 1 1 1 1>}}
  User.PaintedScalar {type=NiagaraFloat, defaultValue=1, description=""}
  User.PortCount {type=NiagaraInt32, defaultValue=2}
 ]
}
